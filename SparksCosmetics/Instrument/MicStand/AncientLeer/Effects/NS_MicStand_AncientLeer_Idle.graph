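# Idle Niagara effect for the Ancient Leer mic stand cosmetic, emitting a sparse stream of glowing GPU sparks that fade from pink-magenta to blue.
# Particles spawn around a skeletal mesh bone or socket in a tapered offset volume, then rise with curl noise, upward gravity and drag.
# The location is resampled each update, steering a line attraction between straight up and the mesh's motion direction, while smoothed mesh speed drives a particle kill check.
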NiagaraSystem /SparksCosmetics/Instrument/MicStand/AncientLeer/Effects/NS_MicStand_AncientLeer_Idle
dependencies {
 usedDataInterfaces [&NiagaraDataInterfaceCurve, &NiagaraDataInterfaceVector2DCurve, &NiagaraDataInterfaceColorCurve]
 usedDynamicInputs []
 usedModules [&SystemState]
}
emitters {Sparks_GPU {
  simTarget=GPUComputeSim
  emitterUpdateScript [
   &EmitterState()
   &SpawnRate(
    Spawn Probability=0.5
    SpawnRate=&RandomRangeFloat(Minimum=4, Maximum=7, Evaluation Type=::NewEnumerator1)
    Use Spawn Probability=false
   )
  ]
  particleSpawnScript [
   &InitializeParticle(
    Color Mode=::NewEnumerator0
    Lifetime=&RandomRangeFloat(Minimum=0.75, Maximum=1.5)
    Sprite Size=&Multiply_Vector2D_ByFloat(
     Float=&FloatFromCurve(
      FloatCurve=DataInterface() {propertyValues {Curve {keys [<at 0 5>,<at 0.2 4>,<at 1 2.5>]}}}
      CurveIndex=&RandomRangeFloat(Minimum=0, Maximum=1)
     )
     Vector2D=&RandomRangeVector2D(Minimum=<1 0.5>, Maximum=<1 1>)
    )
    Sprite Size Mode=::NewEnumerator4
    UsePositionOffset=false
    Write Material Random=true
   )
   /SetVariables(
    Transient.BaseWidth:NiagaraFloat=&RandomRangeFloat(Minimum=0, Maximum=1)
    Transient.BaseWidthMax=10
    Transient.BaseWidthMin=1.5
   ) {isSetParameters=true}
   /SetVariables(Particles.PickOffset:Vector3f=&Add_Vector(
     A=&Lerp_Vector(A=<0 0 0>, B=<0 0 0>, Alpha=$Transient.BaseWidth)
     B=&MakeVector(
      X=&Lerp_Float(
       A=1
       B=20
       Alpha=&FloatFromCurve(
        FloatCurve=DataInterface() {propertyValues {Curve {keys [
            {arriveTangent=0.1562, interpMode=RCIM_Cubic, leaveTangent=0.1562, tangentMode=RCTM_User, time=0, value=0}
            {arriveTangent=0.5786, interpMode=RCIM_Cubic, leaveTangent=0.5786, tangentMode=RCTM_User, time=0.5008, value=0.168}
            {arriveTangent=2.6795, tangentMode=RCTM_User, time=1, value=1}
           ]}}}
        CurveIndex=$Transient.BaseWidth
       )
      )
      Y=&Sine(
       Normalized Angle=$Particles.MaterialRandom
       Period=1
       Scale=&Lerp_Float(A=$Transient.BaseWidthMax, B=$Transient.BaseWidthMin, Alpha=$Transient.BaseWidth)
       Bias=0
      )
      Z=&Cosine(
       Normalized Angle=$Particles.MaterialRandom
       Period=1
       Scale=&Lerp_Float(A=$Transient.BaseWidthMax, B=$Transient.BaseWidthMin, Alpha=$Transient.BaseWidth)
       Bias=0
      )
     )
    )) {isSetParameters=true}
   &SkeletalMeshLocation(
    Bone / Socket Index=2
    MaskSpawnGroup=false
    Mesh Sampling Type=::NewEnumerator5
    Orientation Sampling=::NewEnumerator1
    Position Sampling=::NewEnumerator2
    Sampled Position Offset=$Particles.PickOffset
    Skeletal Mesh=$User.SkeletalMesh
    Socket Sampling Mode=::NewEnumerator5
    Use Sampled Mesh Attribute as Ribbon ID=false
   )
   &ShapeLocation(
    MaskSpawnGroup=false
    Override Seed=false
    Rotation Angle=90
    Rotation Mode=::NewEnumerator1
    Shape Origin=$Particles.SkeletalMeshLocation.SampledPosition
    Sphere Radius=1
    UseOwnerScale=true
   )
   &CurlNoiseForce(Noise Frequency=100, Noise Strength=30, Pan Noise=true, Pan Noise Field=<2 2 0>)
   &AddVelocity(Rotation Coordinate Space=::World, Velocity=<0 0 10>)
   &ApplyInitialForces()
   &DynamicMaterialParameters(
    Index 0 Param 1=&RandomRangeFloat(Minimum=2, Maximum=4)
    Index 0 Param 2=2
    Index 0 Param 3=1
    Module.Param0WriteEnabled=true
    Module.Param1WriteEnabled=true
    Module.Param2WriteEnabled=true
    Module.Param3WriteEnabled=true
   )
  ]
  particleUpdateScript [
   &ParticleState()
   &SkeletalMeshLocation(
    Bone / Socket Index=2
    MaskSpawnGroup=false
    Mesh Sampling Type=::NewEnumerator5
    Orientation Sampling=::NewEnumerator1
    Position Sampling=::NewEnumerator2
    Sampled Position Offset=$Particles.PickOffset
    Skeletal Mesh=$User.SkeletalMesh
    Socket Sampling Mode=::NewEnumerator5
    Use Sampled Mesh Attribute as Ribbon ID=false
   ) {module=SkeletalMeshLocation001}
   &GravityForce(Gravity=<0 0 100>)
   &Drag(Drag=&UniformRangedFloat(Minimum=1.5, Maximum=3), Use Linear Drag=true, Use Rotational Drag=false)
   /SetVariables(
    Transient.VelocityFacing:Vector3f=&NormalizeVector(
     Vector To Normalize=$Particles.SkeletalMeshLocation001.SampledVelocity
     Safe Normalize=true
     Safe Fallback Threshold=1
     Safe Fallback Vector=<0 0 1>
    )
    Transient.VelocityScale:NiagaraFloat=&RemapRange(
     Input Value To Remap=&VectorLength(Vector=$Particles.SkeletalMeshLocation001.SampledVelocity)
     Input Min=0
     Input Max=100
     Output Min=0
     Output Max=1
    )
   ) {isSetParameters=true}
   /Niagara/Modules/Update/Forces/LineAttractionForce(
    Attraction Strength=150
    Line End:Vector3f=&AddVectorToPosition(
     Position=$Particles.SkeletalMeshLocation001.SampledPosition
     Vector=&Multiply_VectorByFloat(Vector=&Lerp_Vector(A=<0 0 1>, B=$Transient.VelocityFacing, Alpha=$Transient.VelocityScale), Float=100)
    )
    Line Start:NiagaraPosition=$Particles.SkeletalMeshLocation001.SampledPosition
    Use Attraction Falloff=false
   )
   &SolveForcesAndVelocity(Clamp Velocity=false, Limit Acceleration=false)
   &ScaleSpriteSize(Scale Factor=&Vector2DFromCurve(
     Vector2Curve=DataInterface() {propertyValues {
       XCurve {keys [
         {arriveTangent=5.9728, interpMode=RCIM_Cubic, leaveTangent=5.9728, tangentMode=RCTM_User, time=0, value=0.1}
         {arriveTangent=-0.1, interpMode=RCIM_Cubic, leaveTangent=-0.1, tangentMode=RCTM_User, time=0.18, value=0.8}
         {interpMode=RCIM_Cubic, tangentMode=RCTM_User, time=1, value=0}
        ]}
       YCurve {keys [
         {arriveTangent=6.2715, interpMode=RCIM_Cubic, leaveTangent=6.2715, tangentMode=RCTM_User, time=0, value=0.1}
         {arriveTangent=-0.1001, interpMode=RCIM_Cubic, leaveTangent=-0.1001, tangentMode=RCTM_User, time=0.18, value=0.8}
         {interpMode=RCIM_Cubic, tangentMode=RCTM_User, time=0.9988, value=0}
        ]}
      }}
     ScaleCurve=&Vector2DFromFloat(Value=&RandomRangeFloat(Minimum=1, Maximum=1.5))
    ))
   &ScaleSpriteSizeBySpeed(Max Scale Factor=<0.5 3>, Min Scale Factor=<0.5 1>, Velocity Threshold=666)
   &ScaleColor(
    Linear Color Curve=DataInterface() {propertyValues {
      AlphaCurve {keys [<at 0.333 1>,<at 1 0>]}
      BlueCurve {keys [<at 0 3.5458>,<at 0.166 5>,<at 0.333 4>,<at 0.75 3>]}
      GreenCurve {keys [<at 0 0.24>,<at 0.166 0.2>,<at 0.333 0>,<at 0.75 0>]}
      RedCurve {keys [<at 0 6>,<at 0.166 2.8378>,<at 0.333 1.6139>,<at 0.75 0.0835>]}
     }}
    Scale Mode=::NewEnumerator2
   )
   &InterpolateOverTime(Rate Of Change=10, Target Float Value:NiagaraFloat=$Transient.VelocityScale)
   &KillParticles(Kill Particles=&SetBoolByFloatComparison(A=$Particles.InterpolateOverTime.InterpolateOverTime_Float.Moving Average, B=1, Comparison Type=::NewEnumerator2))
  ]
  rendererClasses [&NiagaraSpriteRendererProperties]
  renderers [
   {rendererClass=&NiagaraSpriteRendererProperties}
   {rendererClass=&NiagaraSpriteRendererProperties, rendererIndex=1}
  ]
 }}
summary {
 systemName=NS_MicStand_AncientLeer_Idle
 emitters [{emitterName=Sparks_GPU, simTarget=GPUComputeSim, rendererClasses [&NiagaraSpriteRendererProperties]}]
 userVariables [User.SkeletalMesh {
   type=NiagaraDataInterfaceSkeletalMesh
   defaultValue {
    struct=NiagaraExt_VariableValue_DataInterface
    value {dataInterface=&self:NiagaraDataInterfaceSkeletalMesh_0}
   }
  }]
}
