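# Idle Niagara effect for the AncientLeer microphone stand cosmetic: purple flame, glow and sparks attached to the skeletal mesh.
# Flame meshes pass a smoothed position to their material, likely for motion lag.
# Sprite and light glows vary alpha by time of day, with a fixed value in front-end preview.
# Rising GPU sparks use curl noise and line attraction that tilts toward sampled mesh velocity, with a smoothed-speed kill check.
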NiagaraSystem /SparksCosmetics/Instrument/MicStand/AncientLeer/Effects/NS_Mic_AncientLeer_Idle
dependencies {
 usedDataInterfaces [&NiagaraDataInterfaceCurve, &NiagaraDataInterfaceVector2DCurve, &NiagaraDataInterfaceColorCurve]
 usedDynamicInputs []
 usedModules [
  &SystemState
  &self:SystemSpawnScript.SystemScriptSource.SystemScriptGraph.NiagaraNodeAssignment_2.SetVariables_#1
  &self:SystemSpawnScript.SystemScriptSource.SystemScriptGraph.NiagaraNodeAssignment_0.SetVariables_#2
 ]
}
emitters {
 Flame {
  emitterUpdateScript [
   &EmitterState()
   &SpawnBurst_Instantaneous(Use Spawn Probability=false, UseLoopCountLimit=true)
  ]
  particleSpawnScript [
   &InitializeParticle(
    Mesh Scale=&Multiply_VectorByFloat(Vector=<0.5 1.2 1>, Float=0.1)
    Mesh Scale Mode=::NewEnumerator4
    UsePositionOffset=false
    Write Material Random=true
   )
   &InitialMeshOrientation(Mesh Axis To Orient=<0 0 1>, Mesh Orientation Mode=::NewEnumerator2, Rotation=<0 0 -0.25>, Use Rotation Vector=true)
   &CameraOffset(Camera Offset Amount=5)
   &FN_Mesh_Attributes(
    Create Orientation Attribute=true
    Create Scale Attribute=true
    Orientation:Quat4f=&FindQuatBetweenTwoVectors(From Vector=<0 0 1>, To Vector=<0 0 -1>)
    Scale=&Multiply_VectorByFloat(Vector=<1 1 1>, Float=0.08)
    Write to Module Namespace=true
   )
  ]
  particleUpdateScript [
   &ParticleState(Kill Particles When Lifetime Has Elapsed=false)
   &SolveForcesAndVelocity(Clamp Velocity=false, Limit Acceleration=false)
   &SkeletalMeshLocation(
    MaskSpawnGroup=false
    Mesh Sampling Type=::NewEnumerator5
    Orientation Sampling=::NewEnumerator1
    Sampled Position Offset=$System.MicOffset
    Skeletal Mesh=$User.SkeletalMesh
    Socket Sampling Mode=::NewEnumerator5
   )
   &InterpolateOverTime(Rate Of Change=10, Target Vector Value:NiagaraPosition=$Particles.Position, Variable Type=::NewEnumerator2)
   &DynamicMaterialParameters(
    Index 1 Param 1=&MakeFloatFromVector(Vector=$Particles.InterpolateOverTime.InterpolateOverTime_V3_.Moving Average, Channel=::NewEnumerator0)
    Index 1 Param 2=&MakeFloatFromVector(Vector=$Particles.InterpolateOverTime.InterpolateOverTime_V3_.Moving Average, Channel=::NewEnumerator1)
    Index 1 Param 3=&MakeFloatFromVector(Vector=$Particles.InterpolateOverTime.InterpolateOverTime_V3_.Moving Average, Channel=::NewEnumerator2)
    Module.Index1Float0Write=true
    Module.Index1Float1Write=true
    Module.Index1Float2Write=true
    Module.Index1Float3Write=true
    Write Parameter Index 0=::NewEnumerator0
    Write Parameter Index 1=::NewEnumerator1
   )
   &ScaleMeshSize(Scale Factor=&VectorFromFloat(Value=$System.MeshScale))
   /SetVariables(
    Particles.FN_Mesh_Attributes.Scale:Vector3f=&Multiply_VectorByFloat(Vector=$Particles.Initial.FN_Mesh_Attributes.Scale, Float=$System.MeshScale)
    Particles.TopViewOrientation:Quat4f=&MultiplyQuaternion(
     Quaternion A=&OrientMeshToVector(Look At Direction=<0 0 1>, Reference Axis Vector=<1 0 0>)
     Quaternion B=$Particles.FN_Mesh_Attributes.MeshOrientation
    )
    Particles.TopViewPosition:NiagaraPosition=&AddVectorToPosition(
     Position=$Particles.Position
     Vector=&RotateVectorByQuaternion(Vector To Rotate=<0 0 0.8>, Quaternion=$Particles.SkeletalMeshLocation.SampledRotation)
    )
   ) {isSetParameters=true}
   &ScaleColor(Scale Alpha=0.75, ScaleA=true, ScaleRGB=true)
  ]
  rendererClasses [&NiagaraMeshRendererProperties]
  renderers [
   {rendererClass=&NiagaraMeshRendererProperties, rendererIndex=0}
   {rendererClass=&NiagaraMeshRendererProperties, rendererIndex=1}
   {rendererClass=&NiagaraMeshRendererProperties, rendererIndex=2}
  ]
 }
 LightGlows {
  emitterUpdateScript [
   &EmitterState()
   &SpawnBurst_Instantaneous(Use Spawn Probability=false, UseLoopCountLimit=true)
  ]
  particleSpawnScript [
   &InitializeParticle(
    Color=<rgba 1.1408 0 2 1>
    Lifetime=0.5
    Sprite Size Mode=::NewEnumerator3
    Uniform Sprite Size=15
    UsePositionOffset=false
    Write Material Random=true
   )
   &Light_Attributes(Falloff Exponent=1, Radius=1) {enabled=false}
   &CameraOffset(Camera Offset Amount=6)
   /SetVariables(
    Particles.GlowCameraOffset=-2
    Particles.GlowColor=&MultiplyLinearColor(Linear Color=$Particles.Color, Scale Factor=<1 1 1 0.75>)
   ) {isSetParameters=true}
  ]
  particleUpdateScript [
   &ParticleState(
    Kill Particles When Lifetime Has Elapsed=false
    Let Infinitely Lived Particles Die When Emitter Deactivates=true
    Loop Particles Lifetime=true
   )
   /SetVariables(
    Particles.GlowColor=&MultiplyLinearColor(
     Linear Color=$Particles.Initial.GlowColor
     Scale Factor=&MakeVector4FromVectorAndFloat(
      "Vector (RGB)"=<1 1 1>
      "Float (Alpha)"=&MakeCustomFloatFromBool(
       Bool=$User.bIsFrontEndPreview
       True Float=0.75
       False Float=&FNTimeOfDayScalar(Night Scale=0.75, Morning Scale=1, Evening Scale=1)
      )
     )
    )
    Particles.GlowScale:Vector2f=&Multiply_Vector2D_ByFloat(Float=$System.MeshScale, Vector2D=$Particles.SpriteSize)
   ) {isSetParameters=true}
   &SkeletalMeshLocation(
    MaskSpawnGroup=false
    Mesh Sampling Type=::NewEnumerator5
    Orientation Sampling=::NewEnumerator1
    Sampled Position Offset=$System.MicOffset
    Skeletal Mesh=$User.SkeletalMesh
    Socket Sampling Mode=::NewEnumerator5
   )
  ]
  rendererClasses [&NiagaraSpriteRendererProperties, &NiagaraLightRendererProperties]
  renderers [
   {rendererClass=&NiagaraSpriteRendererProperties}
   {rendererClass=&NiagaraLightRendererProperties, rendererIndex=1}
   {rendererClass=&NiagaraSpriteRendererProperties, rendererIndex=2}
   {rendererClass=&NiagaraSpriteRendererProperties, rendererIndex=3}
  ]
 }
 Sparks_GPU {
  simTarget=GPUComputeSim
  emitterUpdateScript [
   &EmitterState()
   &SpawnRate(
    Spawn Probability=0.5
    SpawnRate=&RandomRangeFloat(Minimum=15, Maximum=30, Evaluation Type=::NewEnumerator1)
    Use Spawn Probability=false
   )
  ]
  particleSpawnScript [
   &InitializeParticle(
    Color Mode=::NewEnumerator0
    Lifetime=&RandomRangeFloat(Minimum=0.25, Maximum=1)
    Sprite Size=&Multiply_Vector2D_ByFloat(
     Float=&FloatFromCurve(
      FloatCurve=DataInterface() {propertyValues {Curve {keys [<at 0 5>,<at 0.2 4>,<at 1 2.5>]}}}
      CurveIndex=&RandomRangeFloat(Minimum=0, Maximum=1)
      Scale Curve=&MakeCustomFloatFromBool(Bool=$User.bIsFrontEndPreview, True Float=1, False Float=2)
     )
     Vector2D=&RandomRangeVector2D(Minimum=<1 0.5>, Maximum=<1 1>)
    )
    Sprite Size Mode=::NewEnumerator4
    UsePositionOffset=false
    Write Material Random=true
   )
   &PartitionParticles(Number Of Partitions=4, Partition Mode:ENiagara_PartitionMode_v2=::NewEnumerator5)
   &SkeletalMeshLocation(
    MaskSpawnGroup=false
    Mesh Sampling Type=::NewEnumerator5
    Orientation Sampling=::NewEnumerator1
    Position Sampling=::NewEnumerator2
    Sampled Position Offset=$System.MicOffset
    Skeletal Mesh=$User.SkeletalMesh
    Socket Sampling Mode=::NewEnumerator5
    Use Sampled Mesh Attribute as Ribbon ID=false
   )
   &ShapeLocation(
    MaskSpawnGroup=false
    Override Seed=false
    Rotation Angle=90
    Rotation Mode=::NewEnumerator1
    Shape Origin=$Particles.SkeletalMeshLocation.SampledPosition
    Sphere Radius=4
    UseOwnerScale=true
   )
   &CurlNoiseForce(
    Noise Frequency=100
    Noise Strength=&MaskFloat(
     Mask Source=1
     Mask Parameter=$Particles.PartitionParticles.Partition
     Mask=3
     Masked Float=30
     Passthrough Float=&RandomRangeFloat(Minimum=40, Maximum=60)
    )
    Pan Noise=true
    Pan Noise Field=<2 2 0>
   )
   &AddVelocity(
    Rotation Coordinate Space=::World
    Velocity=<0 0 25>
    Velocity Speed Scale=&RandomRangeFloat(Minimum=1, Maximum=2)
   )
   &ApplyInitialForces()
   &DynamicMaterialParameters(
    Index 0 Param 1=&RandomRangeFloat(Minimum=2, Maximum=4)
    Index 0 Param 2=2
    Index 0 Param 3=1
    Module.Param0WriteEnabled=true
    Module.Param1WriteEnabled=true
    Module.Param2WriteEnabled=true
    Module.Param3WriteEnabled=true
   )
  ]
  particleUpdateScript [
   &ParticleState()
   &SkeletalMeshLocation(
    MaskSpawnGroup=false
    Mesh Sampling Type=::NewEnumerator5
    Orientation Sampling=::NewEnumerator1
    Position Sampling=::NewEnumerator2
    Sampled Position Offset=$System.MicOffset
    Skeletal Mesh=$User.SkeletalMesh
    Socket Sampling Mode=::NewEnumerator5
    Use Sampled Mesh Attribute as Ribbon ID=false
   ) {module=SkeletalMeshLocation001}
   &GravityForce(Gravity=<0 0 100>)
   &Drag(Drag=&UniformRangedFloat(Minimum=1.5, Maximum=3), Use Linear Drag=true, Use Rotational Drag=false)
   /SetVariables(
    Transient.VelocityFacing:Vector3f=&NormalizeVector(
     Vector To Normalize=$Particles.SkeletalMeshLocation001.SampledVelocity
     Safe Normalize=true
     Safe Fallback Threshold=1
     Safe Fallback Vector=<0 0 1>
    )
    Transient.VelocityScale:NiagaraFloat=&RemapRange(
     Input Value To Remap=&VectorLength(Vector=$Particles.SkeletalMeshLocation001.SampledVelocity)
     Input Min=0
     Input Max=100
     Output Min=0
     Output Max=1
    )
   ) {isSetParameters=true}
   /Niagara/Modules/Update/Forces/LineAttractionForce(
    Attraction Strength=150
    Line End:Vector3f=&AddVectorToPosition(
     Position=$Particles.SkeletalMeshLocation001.SampledPosition
     Vector=&Multiply_VectorByFloat(Vector=&Lerp_Vector(A=<0 0 1>, B=$Transient.VelocityFacing, Alpha=$Transient.VelocityScale), Float=100)
    )
    Line Start:NiagaraPosition=$Particles.SkeletalMeshLocation001.SampledPosition
    Use Attraction Falloff=false
   )
   &SolveForcesAndVelocity(Clamp Velocity=false, Limit Acceleration=false)
   &ScaleSpriteSize(Scale Factor=&Vector2DFromCurve(
     Vector2Curve=DataInterface() {propertyValues {
       XCurve {keys [
         {arriveTangent=5.9728, interpMode=RCIM_Cubic, leaveTangent=5.9728, tangentMode=RCTM_User, time=0, value=0.1}
         {arriveTangent=-0.1, interpMode=RCIM_Cubic, leaveTangent=-0.1, tangentMode=RCTM_User, time=0.18, value=0.8}
         {interpMode=RCIM_Cubic, tangentMode=RCTM_User, time=1, value=0}
        ]}
       YCurve {keys [
         {arriveTangent=6.2715, interpMode=RCIM_Cubic, leaveTangent=6.2715, tangentMode=RCTM_User, time=0, value=0.1}
         {arriveTangent=-0.1001, interpMode=RCIM_Cubic, leaveTangent=-0.1001, tangentMode=RCTM_User, time=0.18, value=0.8}
         {interpMode=RCIM_Cubic, tangentMode=RCTM_User, time=0.9988, value=0}
        ]}
      }}
     ScaleCurve=&Vector2DFromFloat(Value=&MaskFloat(
       Mask Source=1
       Mask Parameter=$Particles.PartitionParticles.Partition
       Mask=3
       Masked Float=&RandomRangeFloat(Minimum=1, Maximum=2)
       Passthrough Float=1
      ))
    ))
   &ScaleSpriteSizeBySpeed(Max Scale Factor=<0.5 3>, Min Scale Factor=<0.5 1>, Velocity Threshold=666)
   &ScaleColor(
    Linear Color Curve=DataInterface() {propertyValues {
      AlphaCurve {keys [<at 0.333 0.5>,<at 1 0>]}
      BlueCurve {keys [<at 0 3.5458>,<at 0.166 5>,<at 0.333 4>,<at 0.75 3>]}
      GreenCurve {keys [<at 0 0.24>,<at 0.166 0.2>,<at 0.333 0>,<at 0.75 0>]}
      RedCurve {keys [<at 0 6>,<at 0.166 2.8378>,<at 0.333 1.6139>,<at 0.75 0.0835>]}
     }}
    Scale Mode=::NewEnumerator2
   )
   &InterpolateOverTime(Rate Of Change=10, Target Float Value:NiagaraFloat=$Transient.VelocityScale)
   &KillParticles(Kill Particles=&SetBoolByFloatComparison(A=$Particles.InterpolateOverTime.InterpolateOverTime_Float.Moving Average, B=1, Comparison Type=::NewEnumerator2))
  ]
  rendererClasses [&NiagaraSpriteRendererProperties]
  renderers [
   {rendererClass=&NiagaraSpriteRendererProperties}
   {rendererClass=&NiagaraSpriteRendererProperties, rendererIndex=1}
  ]
 }
}
summary {
 systemName=NS_Mic_AncientLeer_Idle
 emitters [
  {emitterName=Sparks_GPU, simTarget=GPUComputeSim, rendererClasses [&NiagaraSpriteRendererProperties]}
  {emitterName=Flame, rendererClasses [&NiagaraMeshRendererProperties]}
  {emitterName=LightGlows, rendererClasses [&NiagaraSpriteRendererProperties, &NiagaraLightRendererProperties]}
 ]
 userVariables [
  User.bIsFrontEndPreview {type=NiagaraBool}
  User.SkeletalMesh {
   type=NiagaraDataInterfaceSkeletalMesh
   defaultValue {
    struct=NiagaraExt_VariableValue_DataInterface
    value {dataInterface=&self:NiagaraDataInterfaceSkeletalMesh_0}
   }
  }
 ]
}
