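# Physics asset for the AncientLeer mic stand instrument cosmetic, defining simulated bodies for its dangling or decorative leaves.
# Each of the fifteen dyn_leaf bones, plus the base and handle, carries a single capsule body, with leaf capsules offset and rotated along the bone.
# Every leaf is constrained directly to the base with identical tight swing and twist limits for slight motion, while the handle joins the base with looser limits.
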PhysicsAsset /SparksCosmetics/Instrument/MicStand/AncientLeer/Meshes/AncientLeer_MicStand_Physics
bodies {
 base [{shapeType=Capsule, length=2, radius=2}]
 dyn_leaf_01 [{shapeType=Capsule, length=4.0007, radius=2, shapeName=dyn_leaf_01_capsule, center=<3.5 0 0>, rotation=<rot 0 -80.9023 -86.1059>}]
 dyn_leaf_02 [{shapeType=Capsule, length=4.0016, radius=2, shapeName=dyn_leaf_02_capsule, center=<3.7 0 0>, rotation=<rot 0 96.245 -87.2317>}]
 dyn_leaf_03 [{shapeType=Capsule, length=3.4531, radius=2, shapeName=dyn_leaf_03_capsule, center=<3 0 0>, rotation=<rot 0 84.254 -89.0958>}]
 dyn_leaf_04 [{shapeType=Capsule, length=3.4531, radius=2, shapeName=dyn_leaf_04_capsule, center=<3 0 0>, rotation=<rot 0 84.3463 -88.8792>}]
 dyn_leaf_05 [{shapeType=Capsule, length=3.7971, radius=2, shapeName=dyn_leaf_05_capsule, center=<3.8 0 0>, rotation=<rot 0 82.5826 -88.2768>}]
 dyn_leaf_06 [{shapeType=Capsule, length=3.797, radius=2, shapeName=dyn_leaf_06_capsule, center=<3.3 0 0>, rotation=<rot 0 83.0757 -86.7685>}]
 dyn_leaf_07 [{shapeType=Capsule, length=3.7971, radius=2, shapeName=dyn_leaf_07_capsule, center=<3.5 0 0>, rotation=<rot 0 84.0224 -88.1188>}]
 dyn_leaf_08 [{shapeType=Capsule, length=3.7971, radius=2, shapeName=dyn_leaf_08_capsule, center=<3.6 0 0>, rotation=<rot 0 80.8861 -89.1838>}]
 dyn_leaf_09 [{shapeType=Capsule, length=3.8717, radius=2, shapeName=dyn_leaf_09_capsule, center=<3.7 0 0>, rotation=<rot 0 -96.6892 -88.4004>}]
 dyn_leaf_10 [{shapeType=Capsule, length=3.3672, radius=2, shapeName=dyn_leaf_10_capsule, center=<3 0 0>, rotation=<rot 0 82.5661 -86.919>}]
 dyn_leaf_11 [{shapeType=Capsule, length=3.8717, radius=2, shapeName=dyn_leaf_11_capsule, center=<3.5 0 0>, rotation=<rot 0 82.5957 -88.7321>}]
 dyn_leaf_12 [{shapeType=Capsule, length=3.4528, radius=2, shapeName=dyn_leaf_12_capsule, center=<3.1 0 0>, rotation=<rot 0 -96.7848 -89.1115>}]
 dyn_leaf_13 [{shapeType=Capsule, length=4.0016, radius=2, shapeName=dyn_leaf_13_capsule, center=<3.7 0 0>, rotation=<rot 0 -83.0263 -87.7163>}]
 dyn_leaf_14 [{shapeType=Capsule, length=4.0013, radius=2, shapeName=dyn_leaf_14_capsule, center=<3.7 0 0>, rotation=<rot 0 98.5584 -89.1326>}]
 dyn_leaf_15 [{shapeType=Capsule, length=4.0015, radius=2, shapeName=dyn_leaf_15_capsule, center=<3.6 0 0>, rotation=<rot 0 -97.005 -87.6909>}]
 handle [{shapeType=Capsule, length=2, radius=2}]
}
constraints [
 {bone1Name=base, bone2Name=handle, swing1LimitDegrees=45, swing2LimitDegrees=45, twistLimitDegrees=45}
 {bone1Name=dyn_leaf_01, bone2Name=base, swing1LimitDegrees=8, swing2LimitDegrees=5, twistLimitDegrees=4}
 {bone1Name=dyn_leaf_02, bone2Name=base, swing1LimitDegrees=8, swing2LimitDegrees=5, twistLimitDegrees=4}
 {bone1Name=dyn_leaf_03, bone2Name=base, swing1LimitDegrees=8, swing2LimitDegrees=5, twistLimitDegrees=4}
 {bone1Name=dyn_leaf_04, bone2Name=base, swing1LimitDegrees=8, swing2LimitDegrees=5, twistLimitDegrees=4}
 {bone1Name=dyn_leaf_05, bone2Name=base, swing1LimitDegrees=8, swing2LimitDegrees=5, twistLimitDegrees=4}
 {bone1Name=dyn_leaf_06, bone2Name=base, swing1LimitDegrees=8, swing2LimitDegrees=5, twistLimitDegrees=4}
 {bone1Name=dyn_leaf_07, bone2Name=base, swing1LimitDegrees=8, swing2LimitDegrees=5, twistLimitDegrees=4}
 {bone1Name=dyn_leaf_08, bone2Name=base, swing1LimitDegrees=8, swing2LimitDegrees=5, twistLimitDegrees=4}
 {bone1Name=dyn_leaf_09, bone2Name=base, swing1LimitDegrees=8, swing2LimitDegrees=5, twistLimitDegrees=4}
 {bone1Name=dyn_leaf_10, bone2Name=base, swing1LimitDegrees=8, swing2LimitDegrees=5, twistLimitDegrees=4}
 {bone1Name=dyn_leaf_11, bone2Name=base, swing1LimitDegrees=8, swing2LimitDegrees=5, twistLimitDegrees=4}
 {bone1Name=dyn_leaf_12, bone2Name=base, swing1LimitDegrees=8, swing2LimitDegrees=5, twistLimitDegrees=4}
 {bone1Name=dyn_leaf_13, bone2Name=base, swing1LimitDegrees=8, swing2LimitDegrees=5, twistLimitDegrees=4}
 {bone1Name=dyn_leaf_14, bone2Name=base, swing1LimitDegrees=8, swing2LimitDegrees=5, twistLimitDegrees=4}
 {bone1Name=dyn_leaf_15, bone2Name=base, swing1LimitDegrees=8, swing2LimitDegrees=5, twistLimitDegrees=4}
]
